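# Sound cue for the looping lockdown alarm in heist bases.
# A short randomized delay feeds a looping wave player, so each instance starts at a slightly different offset.
# It routes to the gameplay events sound class and uses the heist base alarm attenuation and alarm concurrency assets to control falloff and simultaneous playback.
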SoundCue /HeistBasesGameplay/Items/Environmental/LockdownAlarm/SFX/Alarm_Heist_Lockdown_LP_01_Cue
root=Delay_0
nodes [
 {id=Delay_0, children [WavePlayer_4], properties {delayMax=0.5, delayMin=0.2}}
 {
  id=WavePlayer_4
  properties {bLooping=true, soundWaveAssetPtr=/HeistBasesGameplay/Items/Environmental/LockdownAlarm/SFX/wav/Alarm_Heist_LockDwn_LP_01}
 }
]
properties {
 attenuationSettings=&/Att_Heist_Base_Alarms
 maxDistance=1800
 soundClassObject=&Fort_Events_Gameplay
 volumeMultiplier=1
 concurrencyOverrides {bVolumeScaleCanRelease=true, maxCount=3}
 concurrencySet [&/Concurrency_HeistAlarms]
}
