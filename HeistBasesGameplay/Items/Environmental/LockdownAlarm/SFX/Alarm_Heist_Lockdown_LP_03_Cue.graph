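# Looping lockdown alarm sound cue for Heist bases, the third loop variant by name, routed to the gameplay events sound class.
# A looping wave starts after a short random delay and feeds an enveloper that ramps volume up from silence with flat pitch.
# The cue plays slightly lowered in pitch and volume, using Heist alarm attenuation and a shared alarm concurrency group to limit overlapping instances.
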SoundCue /HeistBasesGameplay/Items/Environmental/LockdownAlarm/SFX/Alarm_Heist_Lockdown_LP_03_Cue
root=Enveloper_2
nodes [
 {id=Delay_0, children [WavePlayer_2], properties {delayMax=2, delayMin=0.1}}
 {
  id=Enveloper_2
  children [Delay_0]
  properties {
   pitchCurve {editorCurveData {keys [{time=0}]}}
   volumeCurve {editorCurveData {keys [
      {arriveTangent=-0.0095, interpMode=RCIM_Cubic, time=0, value=0}
      {arriveTangent=1, interpMode=RCIM_Cubic, leaveTangent=1, time=0.3666, value=0.9121}
      <at 1 1>
     ]}}
  }
 }
 {
  id=WavePlayer_2
  properties {bLooping=true, soundWaveAssetPtr=/HeistBasesGameplay/Items/Environmental/LockdownAlarm/SFX/wav/Alarm_Heist_LockDwn_LP_03}
 }
]
properties {
 attenuationSettings=&/Att_Heist_Base_Alarms
 maxDistance=1800
 pitchMultiplier=0.9
 soundClassObject=&Fort_Events_Gameplay
 volumeMultiplier=0.6
 concurrencyOverrides {bVolumeScaleCanRelease=true, maxCount=3}
 concurrencySet [&/Concurrency_HeistAlarms]
}
