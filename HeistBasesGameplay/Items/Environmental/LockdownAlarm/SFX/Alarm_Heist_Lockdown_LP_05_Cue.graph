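# Looping lockdown alarm sound cue for Heist bases, variant 05 of the lockdown alarm set.
# A random one-to-two-second delay feeds a looping wave player, offsetting when each instance starts.
# Playback is pitched down and routed to the gameplay events sound class, using shared Heist alarm attenuation and a concurrency set to limit overlapping alarms.
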SoundCue /HeistBasesGameplay/Items/Environmental/LockdownAlarm/SFX/Alarm_Heist_Lockdown_LP_05_Cue
root=Delay_0
nodes [
 {id=Delay_0, children [WavePlayer_1], properties {delayMax=2, delayMin=1}}
 {
  id=WavePlayer_1
  properties {bLooping=true, soundWaveAssetPtr=/HeistBasesGameplay/Items/Environmental/LockdownAlarm/SFX/wav/Alarm_Heist_LockDwn_LP_05}
 }
]
properties {
 attenuationSettings=&/Att_Heist_Base_Alarms
 maxDistance=1800
 pitchMultiplier=0.8
 soundClassObject=&Fort_Events_Gameplay
 volumeMultiplier=1
 concurrencyOverrides {bVolumeScaleCanRelease=true, maxCount=3}
 concurrencySet [&/Concurrency_HeistAlarms]
}
