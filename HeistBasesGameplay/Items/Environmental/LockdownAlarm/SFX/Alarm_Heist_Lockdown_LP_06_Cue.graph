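# Looping lockdown alarm sound cue for Heist bases, variant 06, with shared alarm attenuation and concurrency.
# After a random start delay, a mixer layers three branches: two share the looping LP_06 wave with slightly different short delays, one pitched down.
# The third branch loops a quieter LP_01 wave behind a several-second delay, adding a periodic secondary alarm.
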SoundCue /HeistBasesGameplay/Items/Environmental/LockdownAlarm/SFX/Alarm_Heist_Lockdown_LP_06_Cue
root=Delay_0
nodes [
 {id=Delay_0, children [Mixer_0], properties {delayMax=3, delayMin=1}}
 {id=Delay_1, children [WavePlayer_1], properties {delayMax=0.15, delayMin=0.15}}
 {id=Delay_2, children [Modulator_2], properties {delayMax=6, delayMin=6}}
 {id=Delay_3, children [WavePlayer_1], properties {delayMax=0.2, delayMin=0.2}}
 {id=Looping_4, children [Delay_1]}
 {id=Looping_5, children [Delay_2]}
 {id=Mixer_0, children [Modulator_1,Looping_4,Looping_5], properties {inputVolume [1,1,1]}}
 {id=Modulator_1, children [Delay_3], properties {pitchMax=0.5, pitchMin=0.5}}
 {id=Modulator_2, children [WavePlayer_0], properties {pitchMax=1, pitchMin=1, volumeMax=0.68, volumeMin=0.68}}
 {id=WavePlayer_0, properties {soundWaveAssetPtr=/HeistBasesGameplay/Items/Environmental/LockdownAlarm/SFX/wav/Alarm_Heist_LockDwn_LP_01}}
 {
  id=WavePlayer_1
  properties {bLooping=true, soundWaveAssetPtr=/HeistBasesGameplay/Items/Environmental/LockdownAlarm/SFX/wav/Alarm_Heist_LockDwn_LP_06}
 }
]
properties {
 attenuationSettings=&/Att_Heist_Base_Alarms
 maxDistance=1800
 soundClassObject=&Fort_Events_Gameplay
 volumeMultiplier=1
 concurrencyOverrides {bVolumeScaleCanRelease=true, maxCount=3}
 concurrencySet [&/Concurrency_HeistAlarms]
}
